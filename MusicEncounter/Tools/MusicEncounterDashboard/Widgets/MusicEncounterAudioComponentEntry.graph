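# Editor utility list entry widget that represents an audio component in the Music Encounter Dashboard tool, built on TechAudioToolsEditorEntry.
# When its list item object is set, it casts the object to BP_MusicEncounterAudioComponentViewModel and assigns it as the widget's viewmodel.
# The layout is a dark bordered row with a green-tinted audio icon and an "Audio Component" label.
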EditorUtilityWidgetBlueprint /MusicEncounter/Tools/MusicEncounterDashboard/Widgets/MusicEncounterAudioComponentEntry
graphs [EventGraph {nodes [
   n1 = Event OnListItemObjectSet()
   > n2 = Cast To BP_MusicEncounterAudioComponentViewModel!(execute, Object=@n1.ListItemObject) -> (then, CastFailed, AsMusic Encounter Audio Component Viewmodel:Music Encounter Audio Component Viewmodel Object Reference) {is=title K2Node_DynamicCast}
   > SetBP_MusicEncounterAudioComponentViewModel!(execute, self, Viewmodel:Music Encounter Audio Component Viewmodel Object Reference=@n2) -> (then)
  ]}]
variables []
widgets {
 info {parentClass=/Script/TechAudioToolsContentEditor.TechAudioToolsEditorEntry, rootWidgetClass=&VerticalBox, widgetCount=5}
 widgets [
  VerticalBox_0 {widgetClassPath=&VerticalBox, widget=VerticalBox_0}
  Border_0 {
   widgetClassPath=&Border
   parent=VerticalBox_0
   widget=Border_0
   properties {brushColor=<rgba 0.0176 0.0176 0.0176 1>, padding=<ltrb 2 2 2 2>}
   slot {class=VerticalBoxSlot}
  }
  HorizontalBox_27 {
   widgetClassPath=&HorizontalBox
   parent=Border_0
   widget=HorizontalBox_27
   slot {class=BorderSlot, properties {padding=<ltrb 2 2 2 2>}}
  }
  Image_105 {
   widgetClassPath=&Image
   bIsVariable=true
   parent=HorizontalBox_27
   widget=Image_105
   properties {
    brush {
     bIsDynamicallyLoaded=false
     drawAs=Image
     imageType=FullColor
     mirroring=NoMirror
     resourceName=None
     resourceObject=/FortniteCoreUI/UI/Icons/Icon-Settings/T-Icon-Audio-64
     tiling=NoTile
     imageSize=<16 16>
     margin=<ltrb 0 0 0 0>
     outlineSettings {
      bUseBrushTransparency=false
      roundingType=HalfHeightRadius
      width=0
      color {colorUseRule=UseColor_Specified, specifiedColor=<rgba 0 0 0 0>}
      cornerRadii=<0 0 0 0>
     }
     tintColor {colorUseRule=UseColor_Specified, specifiedColor=<rgba 1 1 1 1>}
     uVRegion {bIsValid=false, max=<0 0>, min=<0 0>}
    }
    colorAndOpacity=<rgba 0 0.75 0 1>
   }
   slot {class=HorizontalBoxSlot}
  }
  TextBlock_51 {
   widgetClassPath=&TextBlock
   parent=HorizontalBox_27
   widget=TextBlock_51
   properties {
    text=Audio Component
    colorAndOpacity {colorUseRule=UseColor_Specified, specifiedColor=<rgba 0.5271 0.5271 0.5271 1>}
    font {
     bForceMonospaced=false
     bMaterialIsStencil=false
     fontMaterial=null
     fontObject=&Roboto
     letterSpacing=0
     monospacedWidth=1
     size=9.75
     skewAmount=0
     typefaceFontName=Regular
     outlineSettings {
      bApplyOutlineToDropShadows=false
      bMiteredCorners=false
      bSeparateFillAlpha=false
      outlineMaterial=null
      outlineSize=0
      outlineColor=<rgba 0 0 0 1>
     }
    }
   }
   slot {class=HorizontalBoxSlot}
  }
 ]
}
