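# Boss panel widget for the Music Encounter Dashboard tool, derived from WBP_MusicEncounterInterface.
# Initialize Interface Widget constructs a boss interface viewmodel, assigns it to the widget, then initializes it with the music encounter.
# A vertical stack lays out Boss State, Boss Phase and Targeted Actor rows as label and value text pairs.
# The values hold placeholder-like text, and the graphs here contain no logic that updates them.
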WidgetBlueprint /MusicEncounter/Tools/MusicEncounterDashboard/Widgets/Interfaces/WBP_MusicEncounterInterface_Boss
graphs [
 EventGraph {links [], nodes []}
 Initialize Interface Widget {nodes [
   Initialize Interface Widget!() -> (then, Music Encounter:Music Encounter Base Object Reference) {is=title K2Node_FunctionEntry}
   > n1 = Construct Object from Class(Class=/MusicEncounter/Tools/MusicEncounterDashboard/Viewmodels/InterfaceViewmodels/BP_MusicEncounterInterfaceViewModel_Boss.BP_MusicEncounterInterfaceViewModel_Boss_C) -> (ReturnValue:Music Encounter Interface Viewmodel Boss Object Reference)
   > n2 = Set MusicEncounterInterfaceViewmodel_Boss(MusicEncounterInterfaceViewmodel_Boss:Music Encounter Interface Viewmodel Boss Object Reference=@n1) -> (Output_Get:Music Encounter Interface Viewmodel Boss Object Reference)
   > Initialize Interface!(
    execute
    self:Music Encounter Interface Viewmodel Object Reference=[@n1, @n2]
    Music Encounter:Music Encounter Base Object Reference=Get Music Encounter() -> (Music Encounter:Music Encounter Base Object Reference)
   ) -> (then)
  ]}
]
variables []
widgets {
 info {
  inheritedWidgetCount=0
  parentClass=&/WBP_MusicEncounterInterface.WBP_MusicEncounterInterface_C
  rootWidgetClass=&StackBox
  widgetCount=12
 }
 widgets [
  StackBox_466 {widgetClassPath=&StackBox, widget=StackBox_466, properties {bOverride_Cursor=true, orientation=Orient_Vertical}}
  BossEncounter {widgetClassPath=&HorizontalBox, parent=StackBox_466, widget=BossEncounter, slot {class=StackBoxSlot}}
  BossState {
   widgetClassPath=&HorizontalBox
   parent=BossEncounter
   widget=BossState
   slot {class=HorizontalBoxSlot, properties {size {sizeRule=Fill, value=1}}}
  }
  UEFN_TextBlock_C_51 {
   widgetClassPath=&UEFN_TextBlock
   parent=BossState
   widget=UEFN_TextBlock_C_51
   properties {
    text="Boss State: "
    colorAndOpacity {colorUseRule=UseColor_Specified, specifiedColor=<rgba 0.5271 0.5271 0.5271 1>}
    font {
     bForceMonospaced=false
     bMaterialIsStencil=false
     fontMaterial=null
     fontObject=&Roboto
     letterSpacing=0
     monospacedWidth=1
     size=9.75
     skewAmount=0
     typefaceFontName=Regular
     outlineSettings {
      bApplyOutlineToDropShadows=false
      bMiteredCorners=false
      bSeparateFillAlpha=false
      outlineMaterial=null
      outlineSize=0
      outlineColor=<rgba 0 0 0 1>
     }
    }
   }
   slot {class=HorizontalBoxSlot}
  }
  UEFN_TextBlock_C {
   widgetClassPath=&UEFN_TextBlock
   parent=BossState
   widget=UEFN_TextBlock_C
   properties {
    text=Combat
    font {
     bForceMonospaced=false
     bMaterialIsStencil=false
     fontMaterial=null
     fontObject=&Roboto
     letterSpacing=0
     monospacedWidth=1
     size=9.75
     skewAmount=0
     typefaceFontName=Regular
     outlineSettings {
      bApplyOutlineToDropShadows=false
      bMiteredCorners=false
      bSeparateFillAlpha=false
      outlineMaterial=null
      outlineSize=0
      outlineColor=<rgba 0 0 0 1>
     }
    }
   }
   slot {class=HorizontalBoxSlot}
  }
  BossPhase {widgetClassPath=&HorizontalBox, parent=StackBox_466, widget=BossPhase, slot {class=StackBoxSlot}}
  UEFN_TextBlock_C_1 {
   widgetClassPath=&UEFN_TextBlock
   parent=BossPhase
   widget=UEFN_TextBlock_C_1
   properties {
    text="Boss Phase: "
    colorAndOpacity {colorUseRule=UseColor_Specified, specifiedColor=<rgba 0.5271 0.5271 0.5271 1>}
    font {
     bForceMonospaced=false
     bMaterialIsStencil=false
     fontMaterial=null
     fontObject=&Roboto
     letterSpacing=0
     monospacedWidth=1
     size=9.75
     skewAmount=0
     typefaceFontName=Regular
     outlineSettings {
      bApplyOutlineToDropShadows=false
      bMiteredCorners=false
      bSeparateFillAlpha=false
      outlineMaterial=null
      outlineSize=0
      outlineColor=<rgba 0 0 0 1>
     }
    }
   }
   slot {class=HorizontalBoxSlot}
  }
  UEFN_TextBlock_C_2 {
   widgetClassPath=&UEFN_TextBlock
   parent=BossPhase
   widget=UEFN_TextBlock_C_2
   properties {
    text=1
    font {
     bForceMonospaced=false
     bMaterialIsStencil=false
     fontMaterial=null
     fontObject=&Roboto
     letterSpacing=0
     monospacedWidth=1
     size=9.75
     skewAmount=0
     typefaceFontName=Regular
     outlineSettings {
      bApplyOutlineToDropShadows=false
      bMiteredCorners=false
      bSeparateFillAlpha=false
      outlineMaterial=null
      outlineSize=0
      outlineColor=<rgba 0 0 0 1>
     }
    }
   }
   slot {class=HorizontalBoxSlot}
  }
  Targeting {widgetClassPath=&HorizontalBox, parent=StackBox_466, widget=Targeting, slot {class=StackBoxSlot}}
  TargetedActor {
   widgetClassPath=&HorizontalBox
   parent=Targeting
   widget=TargetedActor
   slot {class=HorizontalBoxSlot, properties {size {sizeRule=Fill, value=1}}}
  }
  UEFN_TextBlock_C_3 {
   widgetClassPath=&UEFN_TextBlock
   parent=TargetedActor
   widget=UEFN_TextBlock_C_3
   properties {
    text="Targeted Actor: "
    colorAndOpacity {colorUseRule=UseColor_Specified, specifiedColor=<rgba 0.5271 0.5271 0.5271 1>}
    font {
     bForceMonospaced=false
     bMaterialIsStencil=false
     fontMaterial=null
     fontObject=&Roboto
     letterSpacing=0
     monospacedWidth=1
     size=9.75
     skewAmount=0
     typefaceFontName=Regular
     outlineSettings {
      bApplyOutlineToDropShadows=false
      bMiteredCorners=false
      bSeparateFillAlpha=false
      outlineMaterial=null
      outlineSize=0
      outlineColor=<rgba 0 0 0 1>
     }
    }
   }
   slot {class=HorizontalBoxSlot}
  }
  UEFN_TextBlock_C_4 {
   widgetClassPath=&UEFN_TextBlock
   parent=TargetedActor
   widget=UEFN_TextBlock_C_4
   properties {
    text=BP_PlayerPawn
    font {
     bForceMonospaced=false
     bMaterialIsStencil=false
     fontMaterial=null
     fontObject=&Roboto
     letterSpacing=0
     monospacedWidth=1
     size=9.75
     skewAmount=0
     typefaceFontName=Regular
     outlineSettings {
      bApplyOutlineToDropShadows=false
      bMiteredCorners=false
      bSeparateFillAlpha=false
      outlineMaterial=null
      outlineSize=0
      outlineColor=<rgba 0 0 0 1>
     }
    }
   }
   slot {class=HorizontalBoxSlot}
  }
 ]
}
