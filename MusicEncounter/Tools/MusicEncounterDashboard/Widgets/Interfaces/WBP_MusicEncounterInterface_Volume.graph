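# Music Encounter dashboard interface widget that displays a "Listener is in Volume" readout for a volume-based music encounter.
# It derives from WBP_MusicEncounterInterface and lays out a grey label beside a value text block in a stack box.
# Initialize Interface Widget constructs a Volume interface viewmodel, sets it on the widget, then initializes that viewmodel with the widget's Music Encounter.
# The event graph is empty, and no graph here writes the value text.
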WidgetBlueprint /MusicEncounter/Tools/MusicEncounterDashboard/Widgets/Interfaces/WBP_MusicEncounterInterface_Volume
graphs [
 EventGraph {links [], nodes []}
 Initialize Interface Widget {nodes [
   Initialize Interface Widget!() -> (then, Music Encounter:Music Encounter Base Object Reference) {is=title K2Node_FunctionEntry}
   > n1 = Construct Object from Class(Class=/MusicEncounter/Tools/MusicEncounterDashboard/Viewmodels/InterfaceViewmodels/BP_MusicEncounterInterfaceViewModel_Volume.BP_MusicEncounterInterfaceViewModel_Volume_C) -> (ReturnValue:Music Encounter Interface Viewmodel Volume Object Reference)
   > SetBP_MusicEncounterInterfaceViewModel_Volume!(execute, self, Viewmodel:Music Encounter Interface Viewmodel Volume Object Reference=@n1) -> (then)
   > Initialize Interface!(
    execute
    self:Music Encounter Interface Viewmodel Object Reference=@n1
    Music Encounter:Music Encounter Base Object Reference=Get Music Encounter() -> (Music Encounter:Music Encounter Base Object Reference)
   ) -> (then)
  ]}
]
variables []
widgets {
 info {
  inheritedWidgetCount=0
  parentClass=&/WBP_MusicEncounterInterface.WBP_MusicEncounterInterface_C
  rootWidgetClass=&StackBox
  widgetCount=5
 }
 widgets [
  StackBox_466 {widgetClassPath=&StackBox, widget=StackBox_466, properties {bOverride_Cursor=true, orientation=Orient_Vertical}}
  VolumeEncounter {widgetClassPath=&HorizontalBox, parent=StackBox_466, widget=VolumeEncounter, slot {class=StackBoxSlot}}
  Volume {
   widgetClassPath=&HorizontalBox
   parent=VolumeEncounter
   widget=Volume
   slot {class=HorizontalBoxSlot, properties {size {sizeRule=Fill, value=1}}}
  }
  UEFN_TextBlock_C_51 {
   widgetClassPath=&UEFN_TextBlock
   parent=Volume
   widget=UEFN_TextBlock_C_51
   properties {
    text="Listener is in Volume:  "
    colorAndOpacity {colorUseRule=UseColor_Specified, specifiedColor=<rgba 0.5271 0.5271 0.5271 1>}
    font {
     bForceMonospaced=false
     bMaterialIsStencil=false
     fontMaterial=null
     fontObject=&Roboto
     letterSpacing=0
     monospacedWidth=1
     size=9.75
     skewAmount=0
     typefaceFontName=Regular
     outlineSettings {
      bApplyOutlineToDropShadows=false
      bMiteredCorners=false
      bSeparateFillAlpha=false
      outlineMaterial=null
      outlineSize=0
      outlineColor=<rgba 0 0 0 1>
     }
    }
   }
   slot {class=HorizontalBoxSlot}
  }
  UEFN_TextBlock_C {
   widgetClassPath=&UEFN_TextBlock
   parent=Volume
   widget=UEFN_TextBlock_C
   properties {
    text=Value
    font {
     bForceMonospaced=false
     bMaterialIsStencil=false
     fontMaterial=null
     fontObject=&Roboto
     letterSpacing=0
     monospacedWidth=1
     size=9.75
     skewAmount=0
     typefaceFontName=Regular
     outlineSettings {
      bApplyOutlineToDropShadows=false
      bMiteredCorners=false
      bSeparateFillAlpha=false
      outlineMaterial=null
      outlineSize=0
      outlineColor=<rgba 0 0 0 1>
     }
    }
   }
   slot {class=HorizontalBoxSlot}
  }
 ]
}
