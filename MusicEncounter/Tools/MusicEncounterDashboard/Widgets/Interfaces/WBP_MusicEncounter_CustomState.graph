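# Widget for the Music Encounter dashboard tool that shows a custom state as a centered icon above a message.
# PreConstruct applies the Message and Icon Brush variables to the text block and image.
# The Set Custom State event overrides both from passed-in text and brush, without updating those variables.
# Helper functions Select Icon Brush and Select Message Text each return one of three supplied options by index, and are not called within this widget.
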WidgetBlueprint /MusicEncounter/Tools/MusicEncounterDashboard/Widgets/Interfaces/WBP_MusicEncounter_CustomState
graphs [
 EventGraph {nodes [
   n1 = Set Custom State!() -> (OutputDelegate, then, InText:Text, InBrush:"Slate Brush Structure (by ref)") {is=title K2Node_CustomEvent}
   > "SetText (Text)"(self=Get Message_Text() -> (Message_Text:Common Text Object Reference), InText=@n1.InText)
   > SetBrush(self=Get Icon() -> (Icon:Image Object Reference), InBrush=@n1.InBrush)
   Event PreConstruct()
   > "SetText (Text)"(self=Get Message_Text(), InText=Get Message() -> (Message:Text))
   > SetBrush(self=Get Icon(), InBrush=Get Icon Brush() -> (Icon Brush:Slate Brush Structure))
  ]}
 Select Icon Brush {nodes [
   n1 = Select Icon Brush!() -> (then, Index:int, Option 0:Slate Brush Structure, Option 1:Slate Brush Structure, Option 2:Slate Brush Structure) {is=title K2Node_FunctionEntry}
   > Return Node!(
    execute
    Slate Brush:Slate Brush Structure=Select!(
     Option 0:Slate Brush Structure=@n1.Option 0
     Option 1:Slate Brush Structure=@n1.Option 1
     Option 2:Slate Brush Structure=@n1.Option 2
     Index:int=@n1
    ) -> (ReturnValue:Slate Brush Structure)
   )
  ]}
 Select Message Text {nodes [
   n1 = Select Message Text!() -> (then, Index:int, Option 0:Text, Option 1:Text, Option 2:Text) {is=title K2Node_FunctionEntry}
   > Return Node!(
    execute
    Message:Text=Select(Option 0:Text=@n1.Option 0, Option 1:Text=@n1.Option 1, Option 2:Text=@n1.Option 2, Index=@n1) -> (ReturnValue:Text)
   )
  ]}
]
variables [
 Icon Brush {
  type=SlateBrush
  default {
   bIsDynamicallyLoaded=false
   drawAs=Image
   imageType=FullColor
   mirroring=NoMirror
   resourceName=None
   resourceObject=/Game/Athena/UI/InGame/Creative/PropertyEditor/Icon_Picker/Textures/128x128/T_UI_IconLibrary_shield_128
   tiling=NoTile
   imageSize=<64 64>
   margin=<ltrb 0 0 0 0>
   outlineSettings {
    bUseBrushTransparency=false
    roundingType=HalfHeightRadius
    width=0
    color {colorUseRule=UseColor_Specified, specifiedColor=<rgba 0 0 0 0>}
    cornerRadii=<0 0 0 0>
   }
   tintColor {colorUseRule=UseColor_Specified, specifiedColor=<rgba 1 1 1 1>}
   uVRegion {bIsValid=false, max=<0 0>, min=<0 0>}
  }
 }
 Message {type=string, default="Custom State\r\nText Goes Here!"}
]
widgets {
 info {inheritedWidgetCount=0, parentClass=&UserWidget, rootWidgetClass=&VerticalBox, widgetCount=4}
 widgets [
  VerticalBox_0 {widgetClassPath=&VerticalBox, widget=VerticalBox_0}
  VerticalBox_33 {
   widgetClassPath=&VerticalBox
   parent=VerticalBox_0
   widget=VerticalBox_33
   slot {class=VerticalBoxSlot, properties {horizontalAlignment=HAlign_Center, verticalAlignment=VAlign_Center}}
  }
  Icon {
   widgetClassPath=&Image
   bIsVariable=true
   parent=VerticalBox_33
   widget=Icon
   properties {brush {
     bIsDynamicallyLoaded=false
     drawAs=Image
     imageType=FullColor
     mirroring=NoMirror
     resourceName=None
     resourceObject=/Game/Athena/UI/InGame/Creative/PropertyEditor/Icon_Picker/Textures_V2/128x128/T_UI_IconLibrary_Clock_128
     tiling=NoTile
     imageSize=<64 64>
     margin=<ltrb 0 0 0 0>
     outlineSettings {
      bUseBrushTransparency=false
      roundingType=HalfHeightRadius
      width=0
      color {colorUseRule=UseColor_Specified, specifiedColor=<rgba 0 0 0 0>}
      cornerRadii=<0 0 0 0>
     }
     tintColor {colorUseRule=UseColor_Specified, specifiedColor=<rgba 1 1 1 1>}
     uVRegion {bIsValid=false, max=<0 0>, min=<0 0>}
    }}
   slot {class=VerticalBoxSlot, properties {horizontalAlignment=HAlign_Center, padding=<ltrb 0 0 0 6>}}
  }
  Message_Text {
   widgetClassPath=&CommonTextBlock
   bIsVariable=true
   parent=VerticalBox_33
   widget=Message_Text
   properties {
    justification=Center
    text="Custom State\r\nText Goes Here!"
    font {
     bForceMonospaced=false
     bMaterialIsStencil=false
     fontMaterial=null
     fontObject=&Roboto
     letterSpacing=0
     monospacedWidth=1
     size=13.5
     skewAmount=0
     typefaceFontName=Bold
     outlineSettings {
      bApplyOutlineToDropShadows=false
      bMiteredCorners=false
      bSeparateFillAlpha=false
      outlineMaterial=null
      outlineSize=0
      outlineColor=<rgba 0 0 0 1>
     }
    }
    shadowOffset=<0 0>
   }
   slot {class=VerticalBoxSlot}
  }
 ]
}
