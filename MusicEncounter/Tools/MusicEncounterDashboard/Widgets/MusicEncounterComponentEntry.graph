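# Editor utility list-entry widget that represents a music encounter component in the MusicEncounter dashboard tool.
# Derives from TechAudioToolsEditorEntry and shows a single "Component Name" label inside a dark bordered row.
# When its list item is set, it casts the object to BP_MusicEncounterComponentViewModel and assigns it as the widget's MVVM viewmodel.
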EditorUtilityWidgetBlueprint /MusicEncounter/Tools/MusicEncounterDashboard/Widgets/MusicEncounterComponentEntry
graphs [EventGraph {nodes [
   n1 = Event OnListItemObjectSet()
   > n2 = Cast To BP_MusicEncounterComponentViewModel!(execute, Object=@n1.ListItemObject) -> (then, CastFailed, AsMusic Encounter Component ViewModel:Music Encounter Component View Model Object Reference) {is=title K2Node_DynamicCast}
   > SetBP_MusicEncounterComponentViewModel!(execute, self, Viewmodel:Music Encounter Component View Model Object Reference=@n2) -> (then)
  ]}]
variables []
widgets {
 info {parentClass=/Script/TechAudioToolsContentEditor.TechAudioToolsEditorEntry, rootWidgetClass=&VerticalBox, widgetCount=4}
 widgets [
  VerticalBox_0 {widgetClassPath=&VerticalBox, widget=VerticalBox_0}
  Border_1 {
   widgetClassPath=&Border
   parent=VerticalBox_0
   widget=Border_1
   properties {brushColor=<rgba 0.0176 0.0176 0.0176 1>, padding=<ltrb 2 2 2 2>}
   slot {class=VerticalBoxSlot}
  }
  HorizontalBox_27 {
   widgetClassPath=&HorizontalBox
   parent=Border_1
   widget=HorizontalBox_27
   slot {class=BorderSlot, properties {padding=<ltrb 2 2 2 2>}}
  }
  TextBlock_51 {
   widgetClassPath=&TextBlock
   parent=HorizontalBox_27
   widget=TextBlock_51
   properties {
    text=Component Name
    colorAndOpacity {colorUseRule=UseColor_Specified, specifiedColor=<rgba 0.5271 0.5271 0.5271 1>}
    font {
     bForceMonospaced=false
     bMaterialIsStencil=false
     fontMaterial=null
     fontObject=&Roboto
     letterSpacing=0
     monospacedWidth=1
     size=9.75
     skewAmount=0
     typefaceFontName=Regular
     outlineSettings {
      bApplyOutlineToDropShadows=false
      bMiteredCorners=false
      bSeparateFillAlpha=false
      outlineMaterial=null
      outlineSize=0
      outlineColor=<rgba 0 0 0 1>
     }
    }
   }
   slot {class=HorizontalBoxSlot}
  }
 ]
}
