EditorUtilityWidgetBlueprint /MusicEncounter/Tools/MusicEncounterDashboard/Widgets/MusicEncounterDetails
graphs [
 Create Interface Widgets {nodes [
   Create Interface Widgets() {is=title K2Node_FunctionEntry}
   > ClearChildren(self=Get MusicEncounterInterfacesBox() -> (MusicEncounterInterfacesBox:Vertical Box Object Reference))
   > n1 = "For Each Loop (Map)"!(
    execute:Exec
    MapPin:Map of Interface Interfaces to Music Encounter Interface Widget Base Class References=Get Interface Widgets() -> (Interface Widgets:Map of Interface Interfaces to Music Encounter Interface Widget Base Class References)
    BreakPin:Exec
   ) -> (then:Exec, KeyPin:Interface Interface, ValuePin:Music Encounter Interface Widget Base Class Reference, CompletedPin:Exec)
   > Branch(Condition=Implements Interface!(
     self:Music Encounter View Model Object Reference=Get BP_MusicEncounterViewModel() -> (BP_MusicEncounterViewModel:Music Encounter View Model Object Reference)
     Interface:Interface Interface=@n1
    ) -> (ReturnValue:bool))
   > n2 = Create Widget!(
    execute:Exec
    WorldContextObject:Object Reference
    Class:User Widget Class Reference=@n1.ValuePin
    OwningPlayer:Player Controller Object Reference
   ) -> (then:Exec, ReturnValue:Music Encounter Interface Widget Base Object Reference)
   > Initialize Interface Widget!(
    execute
    self:Music Encounter Interface Widget Base Object Reference=@n2
    Music Encounter:Music Encounter Base Object Reference=Get Music Encounter!(self:Music Encounter View Model Object Reference=Get BP_MusicEncounterViewModel()) -> (Music Encounter:Music Encounter Base Object Reference) {is=title K2Node_CallFunction}
   ) -> (then)
   > AddChild(self=Get MusicEncounterInterfacesBox(), Content=@n2)
  ]}
 EventGraph {nodes [
   Create Custom State Widgets() {is=title K2Node_CustomEvent}
   > ClearChildren(self=Get CustomState_WrapBox() -> (CustomState_WrapBox:Wrap Box Object Reference))
   > n1 = "For Each Loop (Map)"(MapPin:Map of Interface Interfaces to Music Encounter Interface Custom State Types Structures=Get Custom State Widgets() -> (Custom State Widgets:Map of Interface Interfaces to Music Encounter Interface Custom State Types Structures)) -> (ValuePin:Music Encounter Interface Custom State Types Structure)
   > Branch(Condition=Implements Interface(self=Get BP_MusicEncounterViewModel(), Interface=@n1))
   > n2 = "For Each Loop (Map)"(MapPin:Map of Names to Music Encounter Interface Custom States Structures=Break Music Encounter Interface Custom State Types!(MusicEncounterInterfaceCustomStateTypes:"Music Encounter Interface Custom State Types Structure (by ref)"=@n1.ValuePin) -> (CustomStates_8_#1:Map of Names to Music Encounter Interface Custom States Structures) {is=title K2Node_BreakStruct}) -> (KeyPin:Name, ValuePin:Music Encounter Interface Custom States Structure)
   n3 = Break Music Encounter Interface Custom States!(MusicEncounterInterfaceCustomStates:"Music Encounter Interface Custom States Structure (by ref)"=@n2.ValuePin) -> (CustomState_3_#1:Array of Music Encounter Interface Custom State Structures) {is=title K2Node_BreakStruct}
   > Branch(Condition=Is Valid Index!(TargetArray:Array of Music Encounter Interface Custom State Structures=@n3, IndexToTest:int) -> (ReturnValue:bool))
   n4 = Break Music Encounter Interface Custom State!(MusicEncounterInterfaceCustomState:"Music Encounter Interface Custom State Structure (by ref)"="Get (a copy)"(Array:Array of Music Encounter Interface Custom State Structures=@n3) -> (Output:Music Encounter Interface Custom State Structure)) -> (
    IconBrush_2_#1:Slate Brush Structure
    Message_5_#2:Text
    ToolTip_7_#3:Text
   ) {is=title K2Node_BreakStruct}
   > n5 = Create Widget!(
    execute
    WorldContextObject
    Class=/MusicEncounter/Tools/MusicEncounterDashboard/Widgets/Interfaces/WBP_MusicEncounter_CustomState.WBP_MusicEncounter_CustomState_C
    OwningPlayer
    Icon Brush:Slate Brush Structure=@n4
    Message:Text=@n4.Message_5_#2
   ) -> (then, ReturnValue:Music Encounter Dashboard Custom State Object Reference)
   > AddChild(self=Get CustomState_WrapBox(), Content=@n5)
  ]}
]
variables [
 Custom State Widgets {default {
   /Script/Engine.BlueprintGeneratedClass'/MusicEncounter/Tools/MusicEncounterDashboard/Interfaces/BPI_MusicEncounterDashboard_Boss.BPI_MusicEncounterDashboard_Boss_C' {custom States {
     boss Music State {custom State [
       {
        message=Idle
        tool Tip=""
        icon Brush {
         bIsDynamicallyLoaded=false
         drawAs=Image
         imageType=FullColor
         mirroring=NoMirror
         resourceName=None
         resourceObject=/Game/Athena/UI/InGame/Creative/PropertyEditor/Icon_Picker/Textures/128x128/T_UI_IconLibrary_enemy_128
         tiling=NoTile
         imageSize=<64 64>
         margin=<ltrb 0 0 0 0>
         outlineSettings {
          bUseBrushTransparency=false
          roundingType=HalfHeightRadius
          width=0
          color {colorUseRule=UseColor_Specified, specifiedColor=<rgba 0 0 0 0>}
          cornerRadii=<0 0 0 0>
         }
         tintColor {colorUseRule=UseColor_Specified, specifiedColor=<rgba 0.099 0.099 0.099 1>}
         uVRegion {bIsValid=false, max=<0 0>, min=<0 0>}
        }
       }
       {
        message=Combat
        tool Tip=""
        icon Brush {
         bIsDynamicallyLoaded=false
         drawAs=Image
         imageType=FullColor
         mirroring=NoMirror
         resourceName=None
         resourceObject=&T_UI_IconLibrary_enemy_128
         tiling=NoTile
         imageSize=<64 64>
         margin=<ltrb 0 0 0 0>
         outlineSettings {
          bUseBrushTransparency=false
          roundingType=HalfHeightRadius
          width=0
          color {colorUseRule=UseColor_Specified, specifiedColor=<rgba 0 0 0 0>}
          cornerRadii=<0 0 0 0>
         }
         tintColor {colorUseRule=UseColor_Specified, specifiedColor=<rgba 0.3698 0 0 1>}
         uVRegion {bIsValid=false, max=<0 0>, min=<0 0>}
        }
       }
      ]}
     targeting {custom State [
       {
        message="Listener is\r\nNOT Targeted"
        tool Tip=""
        icon Brush {
         bIsDynamicallyLoaded=false
         drawAs=Image
         imageType=FullColor
         mirroring=NoMirror
         resourceName=None
         resourceObject=/Game/Athena/UI/InGame/Creative/PropertyEditor/Icon_Picker/Textures_V2/128x128/T_UI_IconLibrary_Crosshair_128
         tiling=NoTile
         imageSize=<64 64>
         margin=<ltrb 0 0 0 0>
         outlineSettings {
          bUseBrushTransparency=false
          roundingType=HalfHeightRadius
          width=0
          color {colorUseRule=UseColor_Specified, specifiedColor=<rgba 0 0 0 0>}
          cornerRadii=<0 0 0 0>
         }
         tintColor {colorUseRule=UseColor_Specified, specifiedColor=<rgba 0.099 0.099 0.099 1>}
         uVRegion {bIsValid=false, max=<0 0>, min=<0 0>}
        }
       }
       {
        message="Listener is\r\nTargeted"
        tool Tip=""
        icon Brush {
         bIsDynamicallyLoaded=false
         drawAs=Image
         imageType=FullColor
         mirroring=NoMirror
         resourceName=None
         resourceObject=&T_UI_IconLibrary_Crosshair_128
         tiling=NoTile
         imageSize=<64 64>
         margin=<ltrb 0 0 0 0>
         outlineSettings {
          bUseBrushTransparency=false
          roundingType=HalfHeightRadius
          width=0
          color {colorUseRule=UseColor_Specified, specifiedColor=<rgba 0 0 0 0>}
          cornerRadii=<0 0 0 0>
         }
         tintColor {colorUseRule=UseColor_Specified, specifiedColor=<rgba 0.3698 0 0 1>}
         uVRegion {bIsValid=false, max=<0 0>, min=<0 0>}
        }
       }
      ]}
    }}
   &BlueprintGeneratedClass'/MusicEncounter/Tools/MusicEncounterDashboard/Interfaces/BPI_MusicEncounterDashboard_Volume.BPI_MusicEncounterDashboard_Volume_C' {custom States {inside Volume {custom State [
       {
        message="Listener is NOT\r\nInside Volume"
        tool Tip=""
        icon Brush {
         bIsDynamicallyLoaded=false
         drawAs=Image
         imageType=FullColor
         mirroring=NoMirror
         resourceName=None
         resourceObject=/Game/Athena/UI/InGame/Creative/PropertyEditor/Icon_Picker/Textures_V2/128x128/T_UI_IconLibrary_Cube_128
         tiling=NoTile
         imageSize=<64 64>
         margin=<ltrb 0 0 0 0>
         outlineSettings {
          bUseBrushTransparency=false
          roundingType=HalfHeightRadius
          width=0
          color {colorUseRule=UseColor_Specified, specifiedColor=<rgba 0 0 0 0>}
          cornerRadii=<0 0 0 0>
         }
         tintColor {colorUseRule=UseColor_Specified, specifiedColor=<rgba 0.099 0.099 0.099 1>}
         uVRegion {bIsValid=false, max=<0 0>, min=<0 0>}
        }
       }
       {
        message="Listener is\r\nInside Volume"
        tool Tip=""
        icon Brush {
         bIsDynamicallyLoaded=false
         drawAs=Image
         imageType=FullColor
         mirroring=NoMirror
         resourceName=None
         resourceObject=&T_UI_IconLibrary_Cube_128
         tiling=NoTile
         imageSize=<64 64>
         margin=<ltrb 0 0 0 0>
         outlineSettings {
          bUseBrushTransparency=false
          roundingType=HalfHeightRadius
          width=0
          color {colorUseRule=UseColor_Specified, specifiedColor=<rgba 0 0 0 0>}
          cornerRadii=<0 0 0 0>
         }
         tintColor {colorUseRule=UseColor_Specified, specifiedColor=<rgba 0 0.3698 0.0306 1>}
         uVRegion {bIsValid=false, max=<0 0>, min=<0 0>}
        }
       }
      ]}}}
  }}
 Interface Widgets {default {
   &BlueprintGeneratedClass'/MusicEncounter/Interfaces/Encounters/BPI_MusicEncounter_Boss.BPI_MusicEncounter_Boss_C'=/MusicEncounter/Tools/MusicEncounterDashboard/Widgets/Interfaces/WBP_MusicEncounterInterface_Boss.WBP_MusicEncounterInterface_Boss_C
   &BlueprintGeneratedClass'/MusicEncounter/Interfaces/Encounters/BPI_MusicEncounter_Volume.BPI_MusicEncounter_Volume_C'=/MusicEncounter/Tools/MusicEncounterDashboard/Widgets/Interfaces/WBP_MusicEncounterInterface_Volume.WBP_MusicEncounterInterface_Volume_C
  }}
]
widgets {
 info {rootWidgetClass=&VerticalBox, widgetCount=9}
 widgets [
  VerticalBox_1 {widgetClassPath=&VerticalBox, widget=VerticalBox_1}
  Border_2 {
   widgetClassPath=&Border
   parent=VerticalBox_1
   widget=Border_2
   properties {brushColor=<rgba 0.0176 0.0176 0.0176 1>, padding=<ltrb 4 4 4 4>}
   slot {class=VerticalBoxSlot}
  }
  HorizontalBox_65 {
   widgetClassPath=&HorizontalBox
   parent=Border_2
   widget=HorizontalBox_65
   slot {class=BorderSlot, properties {padding=<ltrb 4 4 4 4>}}
  }
  MusicEncounterWidgetVerticalBox {
   widgetClassPath=&VerticalBox
   bIsVariable=true
   parent=HorizontalBox_65
   widget=MusicEncounterWidgetVerticalBox
   slot {class=HorizontalBoxSlot, properties {size {sizeRule=Fill, value=1}}}
  }
  TextBlock_103 {
   widgetClassPath=&TextBlock
   parent=MusicEncounterWidgetVerticalBox
   widget=TextBlock_103
   properties {
    text=BP_MusicEncounter_Class
    font {
     bForceMonospaced=false
     bMaterialIsStencil=false
     fontMaterial=null
     fontObject=&Roboto
     letterSpacing=0
     monospacedWidth=1
     size=13.5
     skewAmount=0
     typefaceFontName=Bold
     outlineSettings {
      bApplyOutlineToDropShadows=false
      bMiteredCorners=false
      bSeparateFillAlpha=false
      outlineMaterial=null
      outlineSize=0
      outlineColor=<rgba 0 0 0 1>
     }
    }
   }
   slot {class=VerticalBoxSlot, properties {padding=<ltrb 2 2 2 2>}}
  }
  CustomState_WrapBox {
   widgetClassPath=&WrapBox
   bIsVariable=true
   parent=MusicEncounterWidgetVerticalBox
   widget=CustomState_WrapBox
   properties {horizontalAlignment=HAlign_Center, innerSlotPadding=<16 16>}
   slot {class=VerticalBoxSlot, properties {padding=<ltrb 16 16 16 16>}}
  }
  MusicEncounterInterfacesBox {
   widgetClassPath=&VerticalBox
   bIsVariable=true
   parent=MusicEncounterWidgetVerticalBox
   widget=MusicEncounterInterfacesBox
   properties {bOverride_Cursor=true}
   slot {class=VerticalBoxSlot}
  }
  MusicEncounterAudioComponentList {
   widgetClassPath=&/MusicEncounterAudioComponentList
   bIsVariable=true
   parent=VerticalBox_1
   widget=MusicEncounterAudioComponentList
   properties {techAudioToolsListViewModel=&self:WidgetTree.MusicEncounterAudioComponentList.TechAudioToolsListViewModel_0}
   slot {class=VerticalBoxSlot}
  }
  MusicEncounterComponentList {
   widgetClassPath=&/MusicEncounterComponentList
   bIsVariable=true
   parent=VerticalBox_1
   widget=MusicEncounterComponentList
   properties {techAudioToolsListViewModel=&self:WidgetTree.MusicEncounterComponentList.TechAudioToolsListViewModel_0}
   slot {class=VerticalBoxSlot}
  }
 ]
}
