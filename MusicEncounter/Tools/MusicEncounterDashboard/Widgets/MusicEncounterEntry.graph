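# MusicEncounterEntry is an editor utility widget serving as a list entry in the Music Encounter Dashboard, laid out as an icon and name label inside a border.
# When its list item is set, it casts the object to BP_MusicEncounterViewModel and assigns it as its viewmodel.
# Helper functions derive a background color from hover and selection, and a state color from prioritized EMusicEncounterState checks on an integer state, defaulting to red.
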EditorUtilityWidgetBlueprint /MusicEncounter/Tools/MusicEncounterDashboard/Widgets/MusicEncounterEntry
graphs [
 EventGraph {nodes [
   n1 = Event OnListItemObjectSet()
   > n2 = Cast To BP_MusicEncounterViewModel!(execute, Object=@n1.ListItemObject) -> (then, CastFailed, AsMusic Encounter ViewModel:Music Encounter View Model Object Reference) {is=title K2Node_DynamicCast}
   > SetBP_MusicEncounterViewModel!(execute, self, Viewmodel:Music Encounter View Model Object Reference=@n2) -> (then)
  ]}
 Get Background Color {nodes [
   n1 = Get Background Color!() -> (then, Is Hovered:bool=~, Is Selected:bool=~) {is=title K2Node_FunctionEntry}
   > Return Node!(
    execute
    Color:Linear Color Structure=Select(
     Option 0:Linear Color Structure=Select(Option 0=(R=0,G=0,B=0,A=0), Option 1=(R=0.0176,G=0.0176,B=0.0176,A=1), Index=@n1)
     Option 1:Linear Color Structure=(R=0,G=0.162,B=0.7454,A=1)
     Index=@n1.Is Selected
    ) -> (ReturnValue:Linear Color Structure)
   )
  ]}
 Get Encounter State Color {nodes [
   n1 = Get Encounter State Color!() -> (then, State:int) {is=title K2Node_FunctionEntry}
   n2 = Get Music Encounter State Value!(State Type:EMusicEncounterState Enum=NewEnumerator2, State:int=@n1, __WorldContext:Object Reference) -> (ReturnValue:bool) {is=title K2Node_CallFunction}
   n3 = Select(Option 0=(R=1,G=0,B=0,A=1), Option 1=(R=0,G=0.3231,B=1,A=1), Index=Get Music Encounter State Value(State Type=NewEnumerator1, State=@n1))
   > Return Node(Color=Select(
     Option 0=Select(Option 0=@n3, Option 1=(R=0,G=0.75,B=0,A=1), Index=@n2)
     Option 1=(R=0.7231,G=0.7529,B=0,A=1)
     Index=Get Music Encounter State Value(State Type=NewEnumerator3, State=@n1)
    ))
  ]}
]
variables []
widgets {
 info {parentClass=/Script/TechAudioToolsContentEditor.TechAudioToolsEditorEntry, rootWidgetClass=&VerticalBox, widgetCount=6}
 widgets [
  VerticalBox_0 {widgetClassPath=&VerticalBox, widget=VerticalBox_0}
  BackgroundBorder {
   widgetClassPath=&Border
   parent=VerticalBox_0
   widget=BackgroundBorder
   properties {brushColor=<rgba 0.0176 0.0176 0.0176 1>, padding=<ltrb 2 2 2 2>}
   slot {class=VerticalBoxSlot}
  }
  HorizontalBox_27 {
   widgetClassPath=&HorizontalBox
   parent=BackgroundBorder
   widget=HorizontalBox_27
   slot {class=BorderSlot, properties {padding=<ltrb 2 2 2 2>}}
  }
  Image_209 {
   widgetClassPath=&Image
   bIsVariable=true
   parent=HorizontalBox_27
   widget=Image_209
   properties {
    brush {
     bIsDynamicallyLoaded=false
     drawAs=Image
     imageType=FullColor
     mirroring=NoMirror
     resourceName=None
     resourceObject=null
     tiling=NoTile
     imageSize=<16 16>
     margin=<ltrb 0 0 0 0>
     outlineSettings {
      bUseBrushTransparency=false
      roundingType=HalfHeightRadius
      width=0
      color {colorUseRule=UseColor_Specified, specifiedColor=<rgba 0 0 0 0>}
      cornerRadii=<0 0 0 0>
     }
     tintColor {colorUseRule=UseColor_Specified, specifiedColor=<rgba 1 1 1 1>}
     uVRegion {bIsValid=false, max=<0 0>, min=<0 0>}
    }
    colorAndOpacity=<rgba 0 0.75 0 1>
   }
   slot {class=HorizontalBoxSlot, properties {horizontalAlignment=HAlign_Center, verticalAlignment=VAlign_Center}}
  }
  VerticalBox_1 {
   widgetClassPath=&VerticalBox
   parent=HorizontalBox_27
   widget=VerticalBox_1
   slot {class=HorizontalBoxSlot, properties {horizontalAlignment=HAlign_Left, size {sizeRule=Fill, value=1}}}
  }
  TextBlock_51 {
   widgetClassPath=&TextBlock
   parent=VerticalBox_1
   widget=TextBlock_51
   properties {
    text=Music Encounter Name
    colorAndOpacity {colorUseRule=UseColor_Specified, specifiedColor=<rgba 0.5271 0.5271 0.5271 1>}
    font {
     bForceMonospaced=false
     bMaterialIsStencil=false
     fontMaterial=null
     fontObject=&Roboto
     letterSpacing=0
     monospacedWidth=1
     size=9.75
     skewAmount=0
     typefaceFontName=Regular
     outlineSettings {
      bApplyOutlineToDropShadows=false
      bMiteredCorners=false
      bSeparateFillAlpha=false
      outlineMaterial=null
      outlineSize=0
      outlineColor=<rgba 0 0 0 1>
     }
    }
   }
   slot {class=VerticalBoxSlot, properties {padding=<ltrb 2 0 0 0>}}
  }
 ]
}
