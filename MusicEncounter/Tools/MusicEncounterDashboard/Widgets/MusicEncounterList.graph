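# MusicEncounterList is an editor utility widget that provides the list view for the Music Encounter Dashboard tool.
# It subclasses TechAudioToolsEditorListView and configures the inherited ListView to build rows from MusicEncounterEntry widgets, with animated scrolling and click-to-clear selection.
# It adds no graph logic, variables or widgets of its own, so any list behavior comes from the parent class and the entry widget.
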EditorUtilityWidgetBlueprint /MusicEncounter/Tools/MusicEncounterDashboard/Widgets/MusicEncounterList
graphs [EventGraph {links [], nodes []}]
variables []
widgets {
 info {
  inheritedWidgetCount=1
  parentClass=/Script/TechAudioToolsContentEditor.TechAudioToolsEditorListView
  rootWidgetClass=/Script/Blutility.EditorUtilityListView
  widgetCount=1
 }
 widgets [ListView {
   widgetClassPath=&EditorUtilityListView
   bInherited=true
   bIsVariable=true
   widget=ListView
   properties {
    bClearSelectionOnClick=true
    bEnableScrollAnimation=true
    entryWidgetClass=&/MusicEncounterEntry.MusicEncounterEntry_C
    numDesignerPreviewEntries=7
   }
  }]
}
